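# UI material for a season pass header icon, showing either a free or paid Battle Pass icon with an animated-style shine.
# The IsPaid scalar lerps between the FreeIcon and PaidIcon texture parameters.
# The shared ShineSwipe function's alpha is color-dodge blended over the chosen icon to produce the highlight.
# The icon's alpha is masked out separately for the translucent blend.
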Material /SeasonPass/Materials/HeaderIcon/M_UI_NexusHeaderIcon
graph {
 properties {
  bAutomaticallySetUsageInEditor=false
  bUsedWithStaticMesh=false
  blendMode=BLEND_Translucent
  materialDomain=MD_UI
  usedShadingModels=Default Lit
 }
 nodes [
  n1 = LinearInterpolate(
   A=TextureSampleParameter2D() {parameterName=FreeIcon, texture=/BRCosmetics/UI/Foundation/Textures/Icons/Items/T-FNBR-BattlePass-Default}.RGBA
   B=TextureSampleParameter2D() {parameterName=PaidIcon, texture=/Game/UI/Foundation/Textures/Icons/Items/T-FNBR-BattlePass}.RGBA
   Alpha=ScalarParameter() {parameterName=IsPaid, sliderMax=1}
  )
  ComponentMask(None=@n1) {a=true}
  MaterialFunctionCall!(Blend=MaterialFunctionCall!(In_Uv) {materialFunction=/Game/UI/Foundation/Materials/M-Function-ShineSwipe}.Alpha, Base=@n1) {materialFunction=/Engine/Functions/Engine_MaterialFunctions03/Blends/Blend_ColorDodge}
 ]
}
